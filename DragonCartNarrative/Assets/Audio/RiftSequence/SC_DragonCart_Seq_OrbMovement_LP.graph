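# Looping sound cue for orb movement in the DragonCart Rift sequence, routed to the Fort_Events sound class.
# A looping castle-creak structure-movement wave feeds an Enveloper whose volume curve gives a brief fade-in.
# The cue plays at reduced volume through the orb sequence attenuation asset.
# It carries per-owner concurrency override settings with a small voice cap and retrigger time, while its concurrency set holds only a null entry.
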SoundCue /DragonCartNarrative/Assets/Audio/RiftSequence/SC_DragonCart_Seq_OrbMovement_LP
root=Enveloper_0
nodes [
 {
  id=Enveloper_0
  children [WavePlayer_1]
  properties {pitchCurve {editorCurveData {keys [{time=0}]}}, volumeCurve {editorCurveData {keys [<at 0 0>,<at 0.1 1>]}}}
 }
 {
  id=WavePlayer_1
  properties {bLooping=true, soundWaveAssetPtr=/Game/Sounds/Trans_AmbientLoops/StructureMovement/Castle_CreaksA016_Loop_Stereo}
 }
]
properties {
 attenuationSettings=&/ATT_DragonCart_Orb_Seq
 maxDistance=173793.1094
 soundClassObject=&Fort_Events
 volumeMultiplier=0.6
 concurrencyOverrides {bLimitToOwner=true, maxCount=2, retriggerTime=0.25}
 concurrencySet [null]
}
