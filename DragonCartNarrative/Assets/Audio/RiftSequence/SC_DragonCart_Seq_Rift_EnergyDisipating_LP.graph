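# Looping SoundCue for the DragonCart narrative rift sequence, named as the energy-dissipating loop.
# A single looping Wave Player repurposes an Athena Skytube UFO wreckage tunnel energy loop, pitched up slightly at cue level.
# It routes to the Fort_Events sound class and uses the DragonCart orb sequence attenuation for long-range falloff.
# Its concurrency override values describe a small, per-owner voice limit with a retrigger time.
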SoundCue /DragonCartNarrative/Assets/Audio/RiftSequence/SC_DragonCart_Seq_Rift_EnergyDisipating_LP
root=WavePlayer_0
nodes [{
  id=WavePlayer_0
  properties {bLooping=true, soundWaveAssetPtr=/Game/Athena/Sounds/Ambient/Skytube/Skytube_Energy_UFO_Wreckage_Tunnel_1P_LP_01}
 }]
properties {
 attenuationSettings=&/ATT_DragonCart_Orb_Seq
 maxDistance=173793.1094
 pitchMultiplier=1.2
 soundClassObject=&Fort_Events
 volumeMultiplier=1
 concurrencyOverrides {bLimitToOwner=true, maxCount=2, retriggerTime=0.25}
 concurrencySet [null]
}
